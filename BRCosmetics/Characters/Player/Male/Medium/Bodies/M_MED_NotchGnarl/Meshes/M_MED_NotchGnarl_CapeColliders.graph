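# Physics asset of collision shapes for the cape on the male medium NotchGnarl player body, as its name indicates.
# Capsules cover the pelvis, spine, clavicles and neck, with a sphere on the head; spine_02 and spine_05 stack several capsules to fill out the torso.
# Arm, leg and foot bodies carry no shapes, and there are no constraints, so it serves as a collider set rather than a ragdoll.
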PhysicsAsset /BRCosmetics/Characters/Player/Male/Medium/Bodies/M_MED_NotchGnarl/Meshes/M_MED_NotchGnarl_CapeColliders
bodies {
 calf_l []
 calf_r []
 clavicle_l [{shapeType=Capsule, length=9.0192, radius=9.8871, center=<9.0373 3.9908 -2.4002>, rotation=<rot 69.3297 1.4476 176.9462>}]
 clavicle_r [{shapeType=Capsule, length=9.0192, radius=9.8871, center=<-9.0373 -3.9908 2.4002>, rotation=<rot 69.3297 1.4476 -3.0538>}]
 foot_l []
 foot_r []
 head [{shapeType=Sphere, radius=17.3125, center=<11.028 -0.174 0>}]
 lowerarm_l []
 lowerarm_r []
 neck_02 [{shapeType=Capsule, length=13.4375, radius=7.3175, center=<-0.2642 -0.3408 0>, rotation=<rot 90 -171.1193 -175.3978>}]
 pelvis [{shapeType=Capsule, length=8.6456, radius=19.5092, center=<-2.4887 -6.1128 0>}]
 spine_01 [{shapeType=Capsule, length=7.9437, radius=18.1053, center=<-2.1857 -5.3895 0>}]
 spine_02 [
  {shapeType=Capsule, length=7.3264, radius=13.6322, center=<-1.711 1.3231 0.0001>}
  {shapeType=Capsule, length=29.3956, radius=11.5807, center=<2.5024 3.9055 3.9825>, rotation=<rot 90 0 0>}
  {shapeType=Capsule, length=29.875, radius=11.1013, center=<2.4704 3.9054 -3.9825>, rotation=<rot 90 0 0>}
 ]
 spine_03 [{shapeType=Capsule, length=8.2107, radius=13.7031, center=<0.1445 3.5938 0.0001>}]
 spine_04 [{shapeType=Capsule, length=9.4404, radius=14.1325, center=<2.766 6.1893 0.0002>}]
 spine_05 [
  {shapeType=Capsule, length=8.0226, radius=15.695, center=<-1.3837 8.3766 0>}
  {shapeType=Capsule, length=13.9518, radius=13.4435, center=<-0.2916 6.6035 0>}
 ]
 thigh_l []
 thigh_r []
 upperarm_l []
 upperarm_r []
}
constraints []
